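# Library-exposed material function for Niagara FX master materials that fades a material out as its pixels approach the camera.
# An inverted SphereMask between world position and camera position, with parameter-driven radius and hardness, yields the fade factor.
# That factor multiplies Opacity, OpacityMask, BaseColor and EmissiveColor within the incoming Material Attributes, so nearby pixels also darken to black.
# Its description recommends applying it right after the base color texture read.
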MaterialFunction /Game/Effects/Niagara/MasterMaterials/Functions/V2/MF_Skydive_CameraFade_V2
graph {
 properties {
  bExposeToLibrary=true
  description=This alters the opacity and base color on camera distance. THis works best when applied right after base color texture read.
  libraryCategoriesText [FX Opacity]
 }
 nodes [
  n1 = GetMaterialAttributes(None=FunctionInput() {inputName=Material Attributes, inputType=FunctionInput_MaterialAttributes}) {attributeGetTypes [#1]}
  n2 = OneMinus(None=SphereMask(
    A=WorldPosition().XYZ
    B=CameraPositionWS()
    Radius=ScalarParameter() {defaultValue=100, group="6.1) Opacity", parameterName=Camera Distance Falloff, sortPriority=4}
    Hardness=ScalarParameter() {defaultValue=0.2, group="6.1) Opacity", parameterName=Camera Distance Falloff Hardness}
   ))
  n3 = GetMaterialAttributes(None=SetMaterialAttributes!(MaterialAttributes=@n1.MaterialAttributes, Opacity=Multiply(A=@n1.Opacity, B=@n2)) {attributeSetTypes [#1]}) {attributeGetTypes [#2]}
  n4 = GetMaterialAttributes(None=SetMaterialAttributes!(MaterialAttributes=@n3.MaterialAttributes, OpacityMask=Multiply(A=@n3.OpacityMask, B=@n2)) {attributeSetTypes [#2]}) {attributeGetTypes [#3, #4]}
  FunctionOutput(None=SetMaterialAttributes!(MaterialAttributes=@n4.MaterialAttributes, BaseColor=Multiply(A=@n4.BaseColor, B=@n2), EmissiveColor=Multiply(A=@n4.EmissiveColor, B=@n2)) {attributeSetTypes [#3, #4]}) {outputName="Opacitys and BaseColor (MA)"}
 ]
}
